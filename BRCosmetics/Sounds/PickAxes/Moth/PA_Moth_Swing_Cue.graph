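# Sound cue for the Moth cosmetic pickaxe's swing.
# Each quality branch mixes close-attenuated Moth swings with quieter, medium-attenuated distant fire axe swings, using modulators for variation.
# A quality switch steps down from four random Moth waves to three, then to one fixed wave of each layer shared by the last two levels.
# Concurrency override limits it to two instances per owner, stopping the oldest.
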SoundCue /BRCosmetics/Sounds/PickAxes/Moth/PA_Moth_Swing_Cue
root=QualityLevel_0
nodes [
 {id=Attenuation_0, children [Modulator_3], properties {attenuationSettings=/Game/Sounds/Attenuation/Fort_Weapon_Pickaxe_Close_Att}}
 {id=Attenuation_1, children [WavePlayer_39], properties {attenuationSettings=/Game/Sounds/Attenuation/Fort_Weapon_Pickaxe_Medium_Att}}
 {id=Attenuation_2, children [Random_1], properties {attenuationSettings=&Fort_Weapon_Pickaxe_Close_Att}}
 {id=Attenuation_3, children [Modulator_0], properties {attenuationSettings=&Fort_Weapon_Pickaxe_Medium_Att}}
 {id=Attenuation_4, children [Random_4], properties {attenuationSettings=&Fort_Weapon_Pickaxe_Close_Att}}
 {id=Attenuation_5, children [Random_2], properties {attenuationSettings=&Fort_Weapon_Pickaxe_Medium_Att}}
 {id=Mixer_0, children [Attenuation_0, Modulator_6], properties {inputVolume [1,1]}}
 {id=Mixer_1, children [Attenuation_2, Attenuation_3], properties {inputVolume [1.1,1]}}
 {id=Mixer_2, children [Attenuation_4, Modulator_5], properties {inputVolume [1.1,1]}}
 {id=Modulator_0, children [Random_3], properties {pitchMax=1.1, pitchMin=1, volumeMax=0.35, volumeMin=0.35}}
 {id=Modulator_1, children [Mixer_1], properties {pitchMin=0.9, volumeMax=1, volumeMin=0.9}}
 {id=Modulator_2, children [Mixer_2]}
 {id=Modulator_3, children [WavePlayer_17], properties {pitchMax=1.1, pitchMin=0.9, volumeMin=0.9}}
 {id=Modulator_5, children [Attenuation_5], properties {pitchMax=1.1, pitchMin=1, volumeMax=0.35, volumeMin=0.35}}
 {id=Modulator_6, children [Attenuation_1], properties {pitchMax=1.1, pitchMin=1, volumeMax=0.35, volumeMin=0.35}}
 {id=QualityLevel_0, children [Modulator_1,Modulator_2,Mixer_0,Mixer_0]}
 {id=Random_1, children [WavePlayer_0,WavePlayer_3,WavePlayer_6,WavePlayer_7], properties {weights [1,1,1,1]}}
 {
  id=Random_2
  children [WavePlayer_21, WavePlayer_22, WavePlayer_24, WavePlayer_25, WavePlayer_28, WavePlayer_29]
  properties {weights [1,1,1,1,1,1]}
 }
 {
  id=Random_3
  children [WavePlayer_32, WavePlayer_33, WavePlayer_34, WavePlayer_35, WavePlayer_36, WavePlayer_37]
  properties {weights [1,1,1,1,1,1]}
 }
 {id=Random_4, children [WavePlayer_8, WavePlayer_9, WavePlayer_10], properties {weights [1,1,1]}}
 {id=WavePlayer_0, properties {soundWaveAssetPtr=&/PA_Moth_Swing_01}}
 {id=WavePlayer_10, properties {soundWaveAssetPtr=&/PA_Moth_Swing_03}}
 {id=WavePlayer_17, properties {soundWaveAssetPtr=&PA_Moth_Swing_01}}
 {id=WavePlayer_21, properties {soundWaveAssetPtr=/Game/Athena/Sounds/Weapons/Fire_Axe_Swing_Distant_04}}
 {id=WavePlayer_22, properties {soundWaveAssetPtr=/Game/Athena/Sounds/Weapons/Fire_Axe_Swing_Distant_01}}
 {id=WavePlayer_24, properties {soundWaveAssetPtr=/Game/Athena/Sounds/Weapons/Fire_Axe_Swing_Distant_02}}
 {id=WavePlayer_25, properties {soundWaveAssetPtr=/Game/Athena/Sounds/Weapons/Fire_Axe_Swing_Distant_03}}
 {id=WavePlayer_28, properties {soundWaveAssetPtr=/Game/Athena/Sounds/Weapons/Fire_Axe_Swing_Distant_05}}
 {id=WavePlayer_29, properties {soundWaveAssetPtr=/Game/Athena/Sounds/Weapons/Fire_Axe_Swing_Distant_06}}
 {id=WavePlayer_3, properties {soundWaveAssetPtr=&/PA_Moth_Swing_02}}
 {id=WavePlayer_32, properties {soundWaveAssetPtr=&Fire_Axe_Swing_Distant_04}}
 {id=WavePlayer_33, properties {soundWaveAssetPtr=&Fire_Axe_Swing_Distant_01}}
 {id=WavePlayer_34, properties {soundWaveAssetPtr=&Fire_Axe_Swing_Distant_02}}
 {id=WavePlayer_35, properties {soundWaveAssetPtr=&Fire_Axe_Swing_Distant_03}}
 {id=WavePlayer_36, properties {soundWaveAssetPtr=&Fire_Axe_Swing_Distant_05}}
 {id=WavePlayer_37, properties {soundWaveAssetPtr=&Fire_Axe_Swing_Distant_06}}
 {id=WavePlayer_39, properties {soundWaveAssetPtr=&Fire_Axe_Swing_Distant_01}}
 {id=WavePlayer_6, properties {soundWaveAssetPtr=&PA_Moth_Swing_03}}
 {id=WavePlayer_7, properties {soundWaveAssetPtr=&/PA_Moth_Swing_04}}
 {id=WavePlayer_8, properties {soundWaveAssetPtr=&PA_Moth_Swing_01}}
 {id=WavePlayer_9, properties {soundWaveAssetPtr=&PA_Moth_Swing_02}}
]
properties {
 audioPropertiesSheet=/BRCosmetics/Sounds/Settings/PropertySheets/Pickaxe/APS_Cosmetic_Pickaxe_Swing
 bOverrideConcurrency=true
 bPrimeOnLoad=true
 maxDistance=3000
 soundClassObject=&Fort_WeaponMelee_PSM
 volumeMultiplier=0.3
 concurrencyOverrides {bLimitToOwner=true, maxCount=2, resolutionRule=StopOldest}
 concurrencySet [&Player_Melee_Weapon_Conc]
}
